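# Translucent UI material for the Discovery status toggle box, drawing a procedural box with a slanted side.
# Custom SDF nodes union a per-corner rounded box with a parallelogram, then roundness and smoothstep produce an antialiased mask.
# A second, offset parallelogram blends EdgeColor into the base color when the bHasEdge static switch is on.
# A static Flip parameter rotates the UVs half a turn to mirror the shape's direction.
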Material /Game/UI/DiscoveryShared/Status/Materials/M_Box_Toggle
graph {
 properties {
  bUsedWithStaticMesh=false
  blendMode=BLEND_Translucent
  materialDomain=MD_UI
  pixelDepthOffsetMode=PDOM_Legacy
  usedShadingModels=Default Lit
 }
 nodes [
  n1 = MaterialFunctionCall(
   UVs=TextureCoordinate()
   "Rotation Angle (0-1)"=StaticSwitch(True=Constant() {r=0.5}, False=Constant(), Value=StaticBoolParameter() {group=Box, parameterName=Flip, sortPriority=44})
  ) {materialFunction=&CustomRotator}
  n2 = Subtract(
   A=@n1.Rotated Values
   B=AppendVector(
    A=ScalarParameter() {defaultValue=0.675, group=Parallelogram, parameterName=Parallelogram Position X}
    B=ScalarParameter() {defaultValue=0.5, group=Parallelogram, parameterName=Parallelogram Position Y}
   )
  )
  MaterialExpressionAdd_1 = Add(A=@n2, B=Constant2Vector() {r=-0.175}) {constB=0}
  MaterialExpressionScalarParameter_10 = ScalarParameter() {defaultValue=0.075, group=Parallelogram, parameterName=Parallelogram Border Width, sliderMax=1, sortPriority=35}
  MaterialExpressionScalarParameter_11 = ScalarParameter() {defaultValue=0.5, group=Parallelogram, parameterName=Parallelogram Border Height, sliderMax=1, sortPriority=36}
  MaterialExpressionScalarParameter_12 = ScalarParameter() {defaultValue=-0.12, group=Parallelogram, parameterName=Parallelogram Border Skew, sliderMax=1, sliderMin=-1, sortPriority=37}
  n3 = Custom!(
   p=@MaterialExpressionAdd_1
   wi=@MaterialExpressionScalarParameter_10
   he=@MaterialExpressionScalarParameter_11
   sk=@MaterialExpressionScalarParameter_12
  ) {
   code="float2 e = float2(sk,he);\r\n    p = (p.y<0.0)?-p:p;\r\n    float2  w = p - e; w.x -= clamp(w.x,-wi,wi);\r\n    float2  d = float2(dot(w,w), -w.y);\r\n    float s = p.x*e.y - p.y*e.x;\r\n    p = (s<0.0)?-p:p;\r\n    float2  v = p - float2(wi,0); v -= e*clamp(dot(v,e)/dot(e,e),-1.0,1.0);\r\n    d = min( d, float2(dot(v,v), wi*he-abs(s)));\r\n    return sqrt(d.x)*sign(-d.y);"
   description=Parallelogram
   inputs [
    {inputName=p, input {expression=&self:MaterialExpressionAdd_1}}
    {inputName=wi, input {expression=&self:MaterialExpressionScalarParameter_10}}
    {inputName=he, input {expression=&self:MaterialExpressionScalarParameter_11}}
    {inputName=sk, input {expression=&self:MaterialExpressionScalarParameter_12}}
   ]
  }
  NamedRerouteDeclaration(None=LinearInterpolate(
    A=VectorParameter() {group=Box, parameterName=EdgeColor, sortPriority=33, defaultValue=<rgba 0.25 0.25 0.25 1>}.RGB
    B=VectorParameter() {group=Box, parameterName=Color, sortPriority=33, defaultValue=<rgba 1 1 1 1>}.RGB
    Alpha=SmoothStep(Value=@n3) {constMax=0.05}
   )) {name=UseEdge, nodeColor=<rgba 0 0 0 1>}
  MaterialExpressionAdd_0 = Add(
   A=ConstantBiasScale(None=@n1.Rotated Values) {bias=-0.5, scale=2}
   B=AppendVector(
    A=ScalarParameter() {defaultValue=0.35, group=Box, parameterName=Position X, sortPriority=34}
    B=ScalarParameter() {group=Box, parameterName=Position Y, sortPriority=35}
   )
  )
  MaterialExpressionAppendVector_1 = AppendVector(
   A=ScalarParameter() {defaultValue=0.65, group=Box, parameterName=Box Width, sortPriority=36}
   B=ScalarParameter() {defaultValue=1, group=Box, parameterName=Box Height, sortPriority=37}
  )
  MaterialExpressionAppendVector_5 = AppendVector(
   A=AppendVector(
    A=ScalarParameter() {group=Corner Roundness, parameterName=Bottom Right, sliderMax=0.5, sortPriority=41}
    B=ScalarParameter() {group=Corner Roundness, parameterName=Top Right, sliderMax=0.5, sortPriority=39}
   )
   B=AppendVector(
    A=ScalarParameter() {defaultValue=0.25, group=Corner Roundness, parameterName=Bottom Left, sliderMax=0.5, sortPriority=40}
    B=ScalarParameter() {defaultValue=0.25, group=Corner Roundness, parameterName=Top Left, sliderMax=0.5, sortPriority=38}
   )
  )
  n4 = Custom!(p=@MaterialExpressionAdd_0, b=@MaterialExpressionAppendVector_1, r=@MaterialExpressionAppendVector_5) {
   code="r.xy = (p.x>0.0)?r.xy : r.zw;\r\n    r.x  = (p.y>0.0)?r.x  : r.y;\r\n    float2 q = abs(p)-b+r.x;\r\n    return min(max(q.x,q.y),0.0) + length(max(q,0.0)) - r.x;"
   description=Rounded Box
   inputs [
    {inputName=p, input {expression=&self:MaterialExpressionAdd_0}}
    {inputName=b, input {expression=&self:MaterialExpressionAppendVector_1}}
    {inputName=r, input {expression=&self:MaterialExpressionAppendVector_5}}
   ]
  }
  MaterialExpressionReroute_4 = Reroute(None=@n2)
  MaterialExpressionScalarParameter_13 = ScalarParameter() {defaultValue=0.2, group=Parallelogram, parameterName=Parallelogram Width, sliderMax=1, sortPriority=35}
  MaterialExpressionScalarParameter_14 = ScalarParameter() {defaultValue=0.5, group=Parallelogram, parameterName=Parallelogram Height, sliderMax=1, sortPriority=36}
  MaterialExpressionScalarParameter_15 = ScalarParameter() {defaultValue=-0.12, group=Parallelogram, parameterName=Parallelogram Skew, sliderMax=1, sliderMin=-1, sortPriority=37}
  n5 = Custom!(
   p=@MaterialExpressionReroute_4
   wi=@MaterialExpressionScalarParameter_13
   he=@MaterialExpressionScalarParameter_14
   sk=@MaterialExpressionScalarParameter_15
  ) {
   code="float2 e = float2(sk,he);\r\n    p = (p.y<0.0)?-p:p;\r\n    float2  w = p - e; w.x -= clamp(w.x,-wi,wi);\r\n    float2  d = float2(dot(w,w), -w.y);\r\n    float s = p.x*e.y - p.y*e.x;\r\n    p = (s<0.0)?-p:p;\r\n    float2  v = p - float2(wi,0); v -= e*clamp(dot(v,e)/dot(e,e),-1.0,1.0);\r\n    d = min( d, float2(dot(v,v), wi*he-abs(s)));\r\n    return sqrt(d.x)*sign(-d.y);"
   description=Parallelogram
   inputs [
    {inputName=p, input {expression=&self:MaterialExpressionReroute_4}}
    {inputName=wi, input {expression=&self:MaterialExpressionScalarParameter_13}}
    {inputName=he, input {expression=&self:MaterialExpressionScalarParameter_14}}
    {inputName=sk, input {expression=&self:MaterialExpressionScalarParameter_15}}
   ]
  }
  n6 = Min(A=@n4, B=@n5)
  n7 = ScalarParameter() {group=Box, parameterName=Roundness, sortPriority=42}
  OneMinus(None=SmoothStep(
    Min=Constant()
    Max=ScalarParameter() {defaultValue=0.01, group=Box, parameterName=Smoothness, sliderMax=1, sortPriority=43}
    Value=Subtract(A=@n6, B=@n7)
   ))
  StaticSwitchParameter(True=NamedRerouteUsage(), False=VectorParameter() {group=Box, parameterName=Color, sortPriority=33, defaultValue=<rgba 1 1 1 1>}.RGB) {parameterName=bHasEdge}
 ]
}
