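# CommonUI two-segment text toggle widget showing OFF and ON labels in material-backed borders, displaying a boolean state.
# SetToggleState or SwitchToggle updates IsOn, then Refresh stops animations and plays ToggleOn or ToggleOff, also run on Construct.
# These short animations swap text colors and animate border material fill and edge parameters.
# PreConstruct applies configurable label text and enlarges segment minimum sizes and padding for touch input.
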WidgetBlueprint /Game/UI/DiscoveryShared/Status/TextToggle
animations [
 ToggleOn {
  binding_count=4
  length_seconds=0.1
  object=&self:ToggleOn
  sequence {bindings [
    Border_On {
     id=#1
     possessed_class=Border
     tracks [{
       type=MovieSceneWidgetMaterialTrack
       sections [{
         type=MovieSceneParameterSection
         end_frame=2
         start_frame=0
         channels [
          ColorR {type=MovieSceneScriptingFloatChannel, keys [[0,0.0044],[2,0.0467]]}
          ColorG {type=MovieSceneScriptingFloatChannel, keys [[0,0.0252],[2,0.2086]]}
          ColorB {type=MovieSceneScriptingFloatChannel, keys [[0,0.1981],[2,0.7157]]}
          ColorA {type=MovieSceneScriptingFloatChannel, const [1,2]}
          EdgeColorR {type=MovieSceneScriptingFloatChannel, keys [[0,0.6308],[2,0.2051]]}
          EdgeColorG {type=MovieSceneScriptingFloatChannel, keys [[0,0.0343],[2,0.855]]}
          EdgeColorB {type=MovieSceneScriptingFloatChannel, keys [[0,0.1812],[2,0.1946]]}
          EdgeColorA {type=MovieSceneScriptingFloatChannel, const [1,2]}
         ]
        }]
      }]
    }
    CText_On {
     id=#2
     possessed_class=CommonTextBlock
     tracks [{
       type=MovieSceneColorTrack
       display_name=Color and Opacity
       sections [{
         type=MovieSceneColorSection
         end_frame=2
         start_frame=0
         channels [
          Color.R {type=MovieSceneScriptingFloatChannel, keys [[0,0.0685],[2,1]]}
          Color.G {type=MovieSceneScriptingFloatChannel, keys [[0,0.1946],[2,1]]}
          Color.B {type=MovieSceneScriptingFloatChannel, keys [[0,0.7011],[2,1]]}
          Color.A {type=MovieSceneScriptingFloatChannel, const [1,2]}
         ]
        }]
      }]
    }
    CText_Off {
     id=#3
     possessed_class=CommonTextBlock
     tracks [{
       type=MovieSceneColorTrack
       display_name=Color and Opacity
       sections [{
         type=MovieSceneColorSection
         end_frame=2
         start_frame=0
         channels [
          Color.R {type=MovieSceneScriptingFloatChannel, keys [[0,1],[2,0.0685]]}
          Color.G {type=MovieSceneScriptingFloatChannel, keys [[0,1],[2,0.1946]]}
          Color.B {type=MovieSceneScriptingFloatChannel, keys [[0,1],[2,0.7011]]}
          Color.A {type=MovieSceneScriptingFloatChannel, const [1,2]}
         ]
        }]
      }]
    }
    Border_Off {
     id=#4
     possessed_class=Border
     tracks [{
       type=MovieSceneWidgetMaterialTrack
       sections [{
         type=MovieSceneParameterSection
         end_frame=2
         start_frame=0
         channels [
          ColorR {type=MovieSceneScriptingFloatChannel, keys [[0,0.0467],[2,0.0044]]}
          ColorG {type=MovieSceneScriptingFloatChannel, keys [[0,0.2086],[2,0.0252]]}
          ColorB {type=MovieSceneScriptingFloatChannel, keys [[0,0.7157],[2,0.1981]]}
          ColorA {type=MovieSceneScriptingFloatChannel, const [1,2]}
          EdgeColorR {type=MovieSceneScriptingFloatChannel, keys [[0,0.6308],[2,0.2051]]}
          EdgeColorG {type=MovieSceneScriptingFloatChannel, keys [[0,0.0343],[2,0.855]]}
          EdgeColorB {type=MovieSceneScriptingFloatChannel, keys [[0,0.1812],[2,0.1946]]}
          EdgeColorA {type=MovieSceneScriptingFloatChannel, const [1,2]}
         ]
        }]
      }]
    }
   ]}
 }
 ToggleOff {
  binding_count=4
  length_seconds=0.1
  object=&self:ToggleOff
  sequence {bindings [
    CText_On {
     id=#5
     possessed_class=CommonTextBlock
     tracks [{
       type=MovieSceneColorTrack
       display_name=Color and Opacity
       sections [{
         type=MovieSceneColorSection
         end_frame=2
         start_frame=0
         channels [
          Color.R {type=MovieSceneScriptingFloatChannel, keys [[0,1],[2,0.0685]]}
          Color.G {type=MovieSceneScriptingFloatChannel, keys [[0,1],[2,0.1946]]}
          Color.B {type=MovieSceneScriptingFloatChannel, keys [[0,1],[2,0.7011]]}
          Color.A {type=MovieSceneScriptingFloatChannel, const [1,2]}
         ]
        }]
      }]
    }
    Border_On {
     id=#6
     possessed_class=Border
     tracks [{
       type=MovieSceneWidgetMaterialTrack
       sections [{
         type=MovieSceneParameterSection
         end_frame=2
         start_frame=0
         channels [
          ColorR {type=MovieSceneScriptingFloatChannel, keys [[0,0.0467],[2,0.0044]]}
          ColorG {type=MovieSceneScriptingFloatChannel, keys [[0,0.2086],[2,0.0252]]}
          ColorB {type=MovieSceneScriptingFloatChannel, keys [[0,0.7157],[2,0.1981]]}
          ColorA {type=MovieSceneScriptingFloatChannel, const [1,2]}
          EdgeColorR {type=MovieSceneScriptingFloatChannel, keys [[0,0.2051],[2,0.6308]]}
          EdgeColorG {type=MovieSceneScriptingFloatChannel, keys [[0,0.855],[2,0.0343]]}
          EdgeColorB {type=MovieSceneScriptingFloatChannel, keys [[0,0.1946],[2,0.1812]]}
          EdgeColorA {type=MovieSceneScriptingFloatChannel, const [1,2]}
         ]
        }]
      }]
    }
    Border_Off {
     id=#7
     possessed_class=Border
     tracks [{
       type=MovieSceneWidgetMaterialTrack
       sections [{
         type=MovieSceneParameterSection
         end_frame=2
         start_frame=0
         channels [
          ColorR {type=MovieSceneScriptingFloatChannel, keys [[0,0.0044],[2,0.0467]]}
          ColorG {type=MovieSceneScriptingFloatChannel, keys [[0,0.0252],[2,0.2086]]}
          ColorB {type=MovieSceneScriptingFloatChannel, keys [[0,0.1981],[2,0.7157]]}
          ColorA {type=MovieSceneScriptingFloatChannel, const [1,2]}
          EdgeColorR {type=MovieSceneScriptingFloatChannel, keys [[0,0.2051],[2,0.6308]]}
          EdgeColorG {type=MovieSceneScriptingFloatChannel, keys [[0,0.855],[2,0.0343]]}
          EdgeColorB {type=MovieSceneScriptingFloatChannel, keys [[0,0.1946],[2,0.1812]]}
          EdgeColorA {type=MovieSceneScriptingFloatChannel, const [1,2]}
         ]
        }]
      }]
    }
    CText_Off {
     id=#8
     possessed_class=CommonTextBlock
     tracks [{
       type=MovieSceneColorTrack
       display_name=Color and Opacity
       sections [{
         type=MovieSceneColorSection
         end_frame=2
         start_frame=0
         channels [
          Color.R {type=MovieSceneScriptingFloatChannel, keys [[0,0.0685],[2,1]]}
          Color.G {type=MovieSceneScriptingFloatChannel, keys [[0,0.1946],[2,1]]}
          Color.B {type=MovieSceneScriptingFloatChannel, keys [[0,0.7011],[2,1]]}
          Color.A {type=MovieSceneScriptingFloatChannel, const [1,2]}
         ]
        }]
      }]
    }
   ]}
 }
]
graphs [
 EventGraph {nodes [
   Event Construct()
   > Refresh()
   Event PreConstruct()
   > "SetText (Text)"(self=Get CText_Off() -> (CText_Off:Common Text Object Reference), InText=Get OffText() -> (OffText:Text))
   > "SetText (Text)"(self=Get CText_On() -> (CText_On:Common Text Object Reference), InText=Get OnText() -> (OnText:Text))
   > SetTouchVisuals()
  ]}
 Is Toggle On {nodes [
   Is Toggle On() {is=title K2Node_FunctionEntry}
   > Return Node!(execute, Is Toggle On:bool=Get IsOn() -> (IsOn:bool))
  ]}
 Refresh {nodes [
   Refresh() {is=title K2Node_FunctionEntry}
   > StopAllAnimations(self:User Widget Object Reference)
   > PlayAnimation(InAnimation=Select(
     Option 0:Widget Animation Object Reference=Get ToggleOff() -> (ToggleOff:Widget Animation Object Reference)
     Option 1:Widget Animation Object Reference=Get ToggleOn() -> (ToggleOn:Widget Animation Object Reference)
     Index=Get IsOn()
    ) -> (ReturnValue:Widget Animation Object Reference))
  ]}
 SetToggleState {nodes [
   n1 = SetToggleState!() -> (then, IsOn:bool=~) {is=title K2Node_FunctionEntry}
   > Set IsOn(IsOn:bool=@n1) -> (Output_Get:bool)
   > Refresh() {is=title K2Node_CallFunction}
  ]}
 SetTouchVisuals {nodes [
   SetTouchVisuals() {is=title K2Node_FunctionEntry}
   > Branch(Condition=IsUsingTouch(PlayerController=GetOwningPlayer()))
   > SetMinDesiredHeight!(
    execute
    self:Size Box Object Reference=[Get SBox_On() -> (SBox_On:Size Box Object Reference), Get SBox_Off() -> (SBox_Off:Size Box Object Reference)]
    InMinDesiredHeight:float=54
   ) -> (then)
   > SetMinDesiredWidth!(execute, self:Size Box Object Reference=[Get SBox_On(), Get SBox_Off()], InMinDesiredWidth:float=104) -> (then)
   > SetPadding!(
    execute
    self:Border Slot Object Reference=SlotAsBorderSlot!(Widget:Widget Object Reference=Get CText_Off()) -> (ReturnValue:Border Slot Object Reference)
    InPadding_Left:float=20
    InPadding_Top:float=0
    InPadding_Right:float=24
    InPadding_Bottom:float=0
   ) -> (then)
   > SetPadding(self=SlotAsBorderSlot(Widget=Get CText_On()), InPadding_Left=24, InPadding_Top=0, InPadding_Right=20, InPadding_Bottom=0)
  ]}
 SwitchToggle {nodes [
   SwitchToggle() {is=title K2Node_FunctionEntry}
   > SetToggleState!(execute, self, IsOn:bool=NOT Boolean(A=Get IsOn())) -> (then) {is=title K2Node_CallFunction}
  ]}
]
variables [
 IsOn {type=boolean}
 OffText {type=string, default=OFF}
 OnText {type=string, default=ON}
]
widgets {
 info {inheritedWidgetCount=0, parentClass=&CommonUserWidget, rootWidgetClass=&GridPanel, widgetCount=7}
 widgets [
  HBox_Toggle {widgetClassPath=&GridPanel, widget=HBox_Toggle}
  SBox_Off {
   widgetClassPath=&SizeBox
   bIsVariable=true
   parent=HBox_Toggle
   widget=SBox_Off
   properties {bOverride_MinDesiredHeight=true, bOverride_MinDesiredWidth=true, maxDesiredWidth=64, minDesiredHeight=34, minDesiredWidth=84}
   slot {class=GridSlot, properties {nudge=<4 0>, padding=<ltrb -4 0 0 0>}}
  }
  Border_Off {
   widgetClassPath=&Border
   parent=SBox_Off
   widget=Border_Off
   properties {
    horizontalAlignment=HAlign_Center
    verticalAlignment=VAlign_Center
    background {
     bIsDynamicallyLoaded=false
     drawAs=Box
     imageType=NoImage
     mirroring=NoMirror
     resourceName=None
     resourceObject=/Game/UI/DiscoveryShared/Status/Materials/M_Box_Toggle_Left
     tiling=NoTile
     imageSize=<32 32>
     margin=<ltrb 0.5 0 0.5 0>
     outlineSettings {
      bUseBrushTransparency=false
      roundingType=HalfHeightRadius
      width=0
      color {colorUseRule=UseColor_Specified, specifiedColor=<rgba 0 0 0 0>}
      cornerRadii=<0 0 0 0>
     }
     tintColor {colorUseRule=UseColor_Specified, specifiedColor=<rgba 1 1 1 1>}
     uVRegion {bIsValid=false, max=<0 0>, min=<0 0>}
    }
    padding=<ltrb 8 0 12 0>
   }
   slot {class=SizeBoxSlot}
  }
  CText_Off {
   widgetClassPath=&CommonTextBlock
   bIsVariable=true
   parent=Border_Off
   widget=CText_Off
   properties {
    clipping=ClipToBounds
    justification=Center
    mobileFontSizeMultiplier=1.1
    text=OFF
    textTransformPolicy=ToUpper
    font {
     bForceMonospaced=false
     bMaterialIsStencil=false
     fontMaterial=null
     fontObject=&BurbankSmall
     letterSpacing=0
     monospacedWidth=1
     size=18
     skewAmount=0
     typefaceFontName=Black
     outlineSettings {
      bApplyOutlineToDropShadows=false
      bMiteredCorners=false
      bSeparateFillAlpha=false
      outlineMaterial=null
      outlineSize=0
      outlineColor=<rgba 0 0 0 1>
     }
    }
    shadowOffset=<0 0>
   }
   slot {class=BorderSlot, properties {horizontalAlignment=HAlign_Center, verticalAlignment=VAlign_Center, padding=<ltrb 8 0 12 0>}}
  }
  SBox_On {
   widgetClassPath=&SizeBox
   bIsVariable=true
   parent=HBox_Toggle
   widget=SBox_On
   properties {bOverride_MinDesiredHeight=true, bOverride_MinDesiredWidth=true, maxDesiredWidth=64, minDesiredHeight=34, minDesiredWidth=84}
   slot {class=GridSlot, properties {column=2, layer=1, nudge=<-5 0>, padding=<ltrb 0 0 -10 0>}}
  }
  Border_On {
   widgetClassPath=&Border
   parent=SBox_On
   widget=Border_On
   properties {
    horizontalAlignment=HAlign_Center
    verticalAlignment=VAlign_Center
    background {
     bIsDynamicallyLoaded=false
     drawAs=Box
     imageType=NoImage
     mirroring=NoMirror
     resourceName=None
     resourceObject=/Game/UI/DiscoveryShared/Status/Materials/M_Box_Toggle_Right
     tiling=NoTile
     imageSize=<32 32>
     margin=<ltrb 0.5 0 0.5 0>
     outlineSettings {
      bUseBrushTransparency=false
      roundingType=HalfHeightRadius
      width=0
      color {colorUseRule=UseColor_Specified, specifiedColor=<rgba 0 0 0 0>}
      cornerRadii=<0 0 0 0>
     }
     tintColor {colorUseRule=UseColor_Specified, specifiedColor=<rgba 1 1 1 1>}
     uVRegion {bIsValid=false, max=<0 0>, min=<0 0>}
    }
    padding=<ltrb 12 0 8 0>
   }
   slot {class=SizeBoxSlot}
  }
  CText_On {
   widgetClassPath=&CommonTextBlock
   bIsVariable=true
   parent=Border_On
   widget=CText_On
   properties {
    justification=Center
    mobileFontSizeMultiplier=1.1
    text=ON
    textTransformPolicy=ToUpper
    colorAndOpacity {colorUseRule=UseColor_Specified, specifiedColor=<rgba 0.0685 0.1946 0.7011 1>}
    font {
     bForceMonospaced=false
     bMaterialIsStencil=false
     fontMaterial=null
     fontObject=&BurbankSmall
     letterSpacing=0
     monospacedWidth=1
     size=18
     skewAmount=0
     typefaceFontName=Black
     outlineSettings {
      bApplyOutlineToDropShadows=false
      bMiteredCorners=false
      bSeparateFillAlpha=false
      outlineMaterial=null
      outlineSize=0
      outlineColor=<rgba 0 0 0 1>
     }
    }
    shadowOffset=<0 0>
   }
   slot {class=BorderSlot, properties {horizontalAlignment=HAlign_Center, verticalAlignment=VAlign_Center, padding=<ltrb 12 0 8 0>}}
  }
 ]
}
